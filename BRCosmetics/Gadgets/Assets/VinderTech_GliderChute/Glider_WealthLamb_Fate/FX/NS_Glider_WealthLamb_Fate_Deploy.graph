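# Niagara system for the WealthLamb Fate glider's deploy effect, drawing short-lived wind cone meshes on a skeletal mesh.
# A single Wind_Cone emitter scales spawn rate with User.ForwardDot and keeps particles on the sampled mesh location.
# System-level values supply the skeletal mesh, TOD tint, alpha fade and universal-mesh flag.
# Front-end modifiers swap rate, offset, rotation and alpha between gameplay and menus, with the menu offset also depending on mesh type.
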NiagaraSystem /BRCosmetics/Gadgets/Assets/VinderTech_GliderChute/Glider_WealthLamb_Fate/FX/NS_Glider_WealthLamb_Fate_Deploy
dependencies {
 usedDataInterfaces [&NiagaraDataInterfaceSkeletalMesh, &NiagaraDataInterfaceColorCurve, &NiagaraDataInterfaceCurve]
 usedDynamicInputs [
  &Multiply_VectorByFloat
  &LerpLinearColors
  &ColorFromCurve
  &SetBoolByFloatComparison
  &MakeFloatFromVector
  &FloatFromCurve
 ]
 usedModules [
  &self:SystemSpawnScript.SystemScriptSource.SystemScriptGraph.NiagaraNodeAssignment_1.SetVariables_#1
  &SampleSkeletalMesh
  &self:SystemSpawnScript.SystemScriptSource.SystemScriptGraph.NiagaraNodeAssignment_2.SetVariables_#2
  &self:SystemSpawnScript.SystemScriptSource.SystemScriptGraph.NiagaraNodeAssignment_3.SetVariables_#3
  &SystemState
  &self:SystemSpawnScript.SystemScriptSource.SystemScriptGraph.NiagaraNodeAssignment_0.SetVariables_#4
 ]
}
emitters {Wind_Cone {
  emitterUpdateScript [
   &EmitterState()
   &SpawnRate(
    SpawnRate=&FN_FrontEndFloatModifier(Gameplay=&Lerp_Float(A=0, B=8, Alpha=$User.ForwardDot), Front End=8)
    Use Spawn Probability=false
   )
  ]
  particleSpawnScript [
   &InitializeParticle(
    Color=&MakeLinearColorFromVectorAndFloat(
     "Vector (RGB)"=&MakeVectorFromLinearColorRGB(Color=$System.TOD)
     "Float (Alpha)"=&Multiply_Float(A=&RandomRangeFloat(Minimum=0.4, Maximum=0.56), B=1)
    )
    Lifetime=0.35
    Mesh Scale Mode=::NewEnumerator3
    Mesh Uniform Scale=&RandomRangeFloat(Minimum=1.4, Maximum=1.6)
    Sprite Size Mode=::NewEnumerator3
    UsePositionOffset=false
    Write Material Random=true
   )
   &InitialMeshOrientation(
    Orientation Axis=<-1 0 0>
    Rotation=&FN_FrontEndVector3Modifier(Gameplay=<0.12 0 0>, Front End=<0.08 0 0>)
    Use Orientation Vector=true
    Use Rotation Vector=true
   )
  ]
  particleUpdateScript [
   &SkeletalMeshLocation(
    MaskSpawnGroup=false
    Mesh Sampling Type=::NewEnumerator5
    Sampled Position Offset=&FN_FrontEndVector3Modifier(
     Gameplay=<0 0 10>
     Front End=&MakeCustomVectorFromBool(Bool=$System.IsUniversalMesh, True Vector=<0 -5 -3>, False Vector=<0 0 5>)
    )
    Skeletal Mesh=$System.Skeletal Mesh
    Socket Sampling Mode=::NewEnumerator5
   )
   &ParticleState()
   &ScaleColor(
    Scale Alpha=&Multiply_Float(A=$System.AlphaFade, B=&FN_FrontEndFloatModifier(Gameplay=1, Front End=0.8))
    ScaleA=true
    ScaleRGB=true
   )
  ]
  rendererClasses [&NiagaraMeshRendererProperties]
  renderers [{rendererClass=&NiagaraMeshRendererProperties, rendererIndex=0}]
 }}
summary {
 systemName=NS_Glider_WealthLamb_Fate_Deploy
 emitters [{emitterName=Wind_Cone, rendererClasses [&NiagaraMeshRendererProperties]}]
 userVariables [
  User.bIsFrontEnd {type=NiagaraBool}
  User.bIsFrontEndPreview {type=NiagaraBool}
  User.ForwardDot {type=NiagaraFloat, defaultValue=1, description=""}
 ]
}
